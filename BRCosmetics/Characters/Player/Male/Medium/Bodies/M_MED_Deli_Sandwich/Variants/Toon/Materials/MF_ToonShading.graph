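# Material function that toon-shades the Deli Sandwich character's Toon variant, baking lighting into Emissive with black BaseColor.
# Sunlight dotted with a mostly vertex-normal blend, plus DotOffset, indexes curve-atlas ramps that blend close to far by distance.
# The ramp multiplies time-of-day-tinted BaseColor, then adds a highlight curve scaled by time of day.
# A view-angle curve lerps in a Fresnel outline color that fades with distance.
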MaterialFunction /BRCosmetics/Characters/Player/Male/Medium/Bodies/M_MED_Deli_Sandwich/Variants/Toon/Materials/MF_ToonShading
graph {nodes [
  n1 = Constant()
  n2 = CollectionParameter() {
   collection=&FortniteMaterialParameters
   parameterName=SunlightDirectionalVector
  }
  n3 = ComponentMask(None=@n2) {b=true, g=true, r=true}
  n4 = GetMaterialAttributes(None=FunctionInput() {inputName=Material, inputType=FunctionInput_MaterialAttributes}) {attributeGetTypes [#1, #2, #3]}
  n5 = ComponentMask(None=VertexNormalWS()) {b=true, g=true, r=true}
  n6 = LinearInterpolate(A=ComponentMask(None=Transform(None=@n4.Normal)) {b=true, g=true, r=true}, B=@n5, Alpha=Constant() {r=0.8})
  n7 = ShadingPathSwitch(Default=@n6, Deferred=@n6)
  n8 = Add(
   A=ConstantBiasScale(None=DotProduct(A=@n3, B=@n7))
   B=FunctionInput() {desc=forDeliSandwich, description=forDeliSandwich, inputName=DotOffset, inputType=FunctionInput_Scalar}
  )
  n9 = CurveAtlasRowParameter(CurveTime=@n8) {
   atlas=&AthenaCurveAtlas01
   curve=/Game/Characters/Player/Male/Medium/Bodies/M_MED_Deli_Sandwich/Variants/Toon/Materials/Curve_Deli_SandwichToon_
   defaultValue=248
   desc=ToonCurveClose
   group=Curves
   parameterName=ToonCurveClose
   sortPriority=1
  }
  n10 = ChannelMaskParameter(None=@n9) {group=Curves, parameterName=ToonCurveClose - Channel, sortPriority=1}
  n11 = CurveAtlasRowParameter(CurveTime=@n8) {
   atlas=&AthenaCurveAtlas01
   curve=&Curve_Deli_SandwichToon_
   defaultValue=248
   desc=ToonCurveFarFade
   group=Curves
   parameterName=ToonCurveFar
   sortPriority=2
  }
  n12 = ChannelMaskParameter(None=@n11) {group=Curves, maskChannel=Blue, parameterName=ToonCurveFar - Channel, sortPriority=3, defaultValue=<rgba 0 0 1 0>}
  n13 = MaterialFunctionCall!(Blend Range=Constant() {desc=spread, r=4000}, Start Offset=Constant() {desc=start, r=-1500}) {materialFunction=/Engine/Functions/Engine_MaterialFunctions03/Particles/Distance_Blend}
  n14 = Multiply(
   A=Saturate(None=MaterialFunctionCall() {materialFunction=/Game/Effects/Fort_Effects/Materials/_Functions/CurveColor_Time_of_Day}.Result)
   B=@n4.BaseColor
  )
  n15 = ScalarParameter() {defaultValue=0.5, parameterName=Day Color Tint Strength}
  n16 = CurveAtlasRowParameter(CurveTime=@n8) {
   atlas=&AthenaCurveAtlas01
   curve=&Curve_Deli_SandwichToon_
   defaultValue=248
   group=Curves
   parameterName=ToonCurveHighlight
   sortPriority=3
  }
  n17 = ChannelMaskParameter(None=@n16) {group=Curves, maskChannel=Green, parameterName=ToonCurveHighlight - Channel, sortPriority=3, defaultValue=<rgba 0 1 0 0>}
  n18 = Multiply(A=@n17.R, B=MaterialFunctionCall() {materialFunction=&CurveColor_Time_of_Day}.Result) {constB=0.1}
  n19 = Constant() {r=0.2}
  n20 = MaterialFunctionCall(Value_Morning=@n19, Value_Day=Constant() {r=0.4}, Value_Evening=@n19, Value_Night=Constant() {r=0.1}) {materialFunction=&Scalar_Time_of_Day}
  n21 = Add(
   A=Multiply(A=LinearInterpolate(A=@n4.BaseColor, B=@n14, Alpha=@n15), B=LinearInterpolate(A=@n10.R, B=@n12.R, Alpha=@n13.Result))
   B=Saturate(None=Multiply(A=@n18, B=@n20.Result) {constB=0.5})
  )
  n22 = VectorParameter() {group=Fresnel Outline, parameterName=Fresnel Outline Color, sortPriority=1, defaultValue=<rgba 0.3854 0.3854 0.3854 1>}
  n23 = ComponentMask(None=CameraVectorWS()) {b=true, g=true, r=true}
  n24 = ShadingPathSwitch(Default=@n23, Deferred=@n23)
  n25 = LinearInterpolate(
   A=ScalarParameter() {defaultValue=0.18, parameterName=Refraction Range - Max}
   B=ScalarParameter() {defaultValue=-0.5, parameterName=Refraction Range - Min}
   Alpha=Saturate(None=DotProduct(A=@n5, B=@n24))
  )
  n26 = CurveAtlasRowParameter(CurveTime=Multiply(A=Saturate(None=@n25), B=ScalarParameter() {defaultValue=10, group=Fresnel Outline, parameterName=Fresnel Outline Opacity})) {
   atlas=&AthenaCurveAtlas01
   curve=&Curve_Deli_SandwichToon_
   defaultValue=248
   group=Fresnel Outline
   parameterName=ToonCurveFresnelOutline
   sortPriority=5
  }
  n27 = ChannelMaskParameter(None=@n26.G) {group=Fresnel Outline, maskChannel=Green, parameterName=FresnelOutlineCurve - Channel, sortPriority=6, defaultValue=<rgba 0 1 0 0>}
  n28 = Multiply(A=@n27.R, B=OneMinus(None=@n13.Result))
  n29 = SetMaterialAttributes!(
   MaterialAttributes
   BaseColor=@n1
   Metallic=@n1
   Specular=@n1
   Roughness=Constant() {r=1}
   EmissiveColor=Saturate(None=LinearInterpolate(A=@n21, B=@n22.RGB, Alpha=@n28))
   Normal=@n4.Normal
  ) {attributeSetTypes [
    #1
    #4
    #5
    #6
    #2
    #3
   ]}
  FunctionOutput(None=@n29)
 ]}
