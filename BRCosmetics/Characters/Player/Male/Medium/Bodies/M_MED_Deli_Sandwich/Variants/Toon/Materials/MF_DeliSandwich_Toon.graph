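# Material function that applies toon shading to the Deli Sandwich character's toon variant and layers emissive effects on top.
# A selectable channel of the logo mask, scaled by LogoOffset, feeds the DotOffset of MF_ToonShading, offsetting shading in the logo area.
# A static switch optionally adds tinted logo glow to the toon emissive.
# The incoming emissive, raised to EyeGradientPower, is then added back into the output.
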MaterialFunction /BRCosmetics/Characters/Player/Male/Medium/Bodies/M_MED_Deli_Sandwich/Variants/Toon/Materials/MF_DeliSandwich_Toon
graph {nodes [
  n1 = TextureSampleParameter2D() {
   group=LogoFX
   parameterName=Logo Mask
   samplerType=SAMPLERTYPE_Masks
   texture=/BRCosmetics/Characters/Player/Male/Medium/Bodies/M_MED_Deli_Sandwich/Variants/Toon/Textures/T_M_MED_Deli_Sandwich_Toon_Body_FX
  }
  n2 = ChannelMaskParameter(None=@n1.RGB) {group=LogoFX, maskChannel=Green, parameterName=LogoMaskChannel, sortPriority=1, defaultValue=<rgba 0 1 0 0>}
  n3 = ScalarParameter() {group=DeliSandwich, parameterName=LogoOffset, sliderMax=0.1, sliderMin=-0.1}
  n4 = GetMaterialAttributes(None=FunctionInput() {inputName=Material, inputType=FunctionInput_MaterialAttributes}) {attributeGetTypes [#1]}
  n5 = GetMaterialAttributes(None=MaterialFunctionCall!(DotOffset=Multiply(A=@n2.R, B=@n3) {constB=0.05}, Material=@n4.MaterialAttributes) {materialFunction=&/MF_ToonShading}.Result) {attributeGetTypes [#1, #2, #3]}
  n6 = Multiply(A=VectorParameter() {group=LogoFX, parameterName=LogoColor, sortPriority=2, defaultValue=<rgba 0.526 0 0.0319 1>}.RGB, B=@n2.R)
  n7 = ScalarParameter() {defaultValue=0.5, group=DeliSandwich, parameterName=Logo Ems Mult, sortPriority=3}
  n8 = StaticSwitchParameter(True=Add(A=@n5.EmissiveColor, B=Multiply(A=@n6, B=@n7)), False=@n5.EmissiveColor) {defaultValue=true, group=DeliSandwich, parameterName=Use Logo FX, sortPriority=0}
  n9 = Power(Base=@n4.EmissiveColor, Exp=ScalarParameter() {defaultValue=1.3, group=DeliSandwich, parameterName=EyeGradientPower}) {constExponent=1.3}
  FunctionOutput(None=SetMaterialAttributes!(MaterialAttributes=@n5.MaterialAttributes, EmissiveColor=Add(A=@n8, B=@n9)) {attributeSetTypes [#1]})
 ]}
