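# Master material for the Toon variant of the M_MED_Deli_Sandwich player character body, used on skeletal meshes.
# Material attributes pass from the shared character PreFX function through MF_DeliSandwich_Toon into the shared PostFX function.
# A UseToonLines static switch, on by default, adds MF_ToonLines borrowed from the Myth character's toon variant.
# Shading model comes from the graph, using Default Lit and Subsurface Profile with a hero skin profile.
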Material /BRCosmetics/Characters/Player/Male/Medium/Bodies/M_MED_Deli_Sandwich/Variants/Toon/Materials/M_MED_Deli_Sandwich_MASTER
graph {
 properties {
  bUseMaterialAttributes=true
  bUsedWithMeshDeformer=true
  bUsedWithMorphTargets=true
  bUsedWithSkeletalMesh=true
  pixelDepthOffsetMode=PDOM_Legacy
  shadingModel=MSM_FromMaterialExpression
  subsurfaceProfile=&SS_HeroSkin_02
  usedShadingModels=Default Lit | Subsurface Profile
 }
 nodes [
  n1 = MaterialFunctionCall() {materialFunction=&BaseFN_CharacterFunction_PreFX}
  n2 = MaterialFunctionCall(
   AO=@n1.AO
   Distance Blend=@n1.AO
   Fresnel=@n1.AO
   Gmap=@n1.AO
   Material=MaterialFunctionCall!(Material=@n1.Result) {materialFunction=&/MF_DeliSandwich_Toon}.Result
  ) {materialFunction=&BaseFN_CharacterFunction_PostFX}
  StaticSwitchParameter(
   True=MaterialFunctionCall!(Material=@n2.Result) {materialFunction=/Game/Characters/Player/Male/Medium/Bodies/M_MED_Myth/Variants/Toon/Materials/MF_ToonLines}.Result
   False=@n2.Result
  ) {defaultValue=true, group=Cartoon, parameterName=UseToonLines}
 ]
}
